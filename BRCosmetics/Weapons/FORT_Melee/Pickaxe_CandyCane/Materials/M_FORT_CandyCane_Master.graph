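# Master clear coat material for the candy cane pickaxe cosmetic, with kill-counter lights and animated lightshows.
# Diffuse, packed mask and normal textures drive the surface attributes, which pass through the TechArt accessories override function.
# Emissive thresholds UV1 by KillCounter over TotalLights, adds sine-gated flashes from a borrowed Pickaxe_15 texture on stepped panning UVs, and is masked by vertex red.
# Lightshow1 and Lightshow2 blend to that texture's green and blue patterns, with Lightshow1 on by default.
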Material /BRCosmetics/Weapons/FORT_Melee/Pickaxe_CandyCane/Materials/M_FORT_CandyCane_Master
graph {
 properties {
  bUseMaterialAttributes=true
  bUsedWithMeshDeformer=true
  bUsedWithSkeletalMesh=true
  pixelDepthOffsetMode=PDOM_Legacy
  shadingModel=MSM_ClearCoat
  usedShadingModels=Clear Coat
 }
 nodes [
  n1 = TextureSample() {samplerType=SAMPLERTYPE_Masks, texture=/BRCosmetics/Weapons/FORT_Melee/Pickaxe_CandyCane/Textures/T_Pickaxe_Candy_Cane_S}
  n2 = Constant() {r=0.381}
  n3 = Constant() {r=1}
  n4 = VertexColor()
  n5 = Add(A=@n4.R, B=@n4.R)
  n6 = ComponentMask(None=TextureCoordinate() {coordinateIndex=1}) {g=true}
  n7 = ScalarParameter() {defaultValue=0.1, parameterName=KillCounter, sliderMax=15, sliderMin=0.1}
  n8 = ScalarParameter() {defaultValue=15, desc=THERE ARE 4 LIGHTS, parameterName=TotalLights}
  n9 = Floor(None=Add(A=Frac(None=@n6), B=Divide(A=@n7, B=@n8)))
  n10 = ScalarParameter() {defaultValue=0.5, parameterName=EmissiveMult}
  n11 = Multiply(A=Time()) {constB=0.2}
  n12 = Constant()
  n13 = Constant() {r=2}
  n14 = Add(A=Floor(None=MaterialFunctionCall!(Sine Phase=@n11, Value 1=@n12, Value 2=@n13) {materialFunction=/Engine/Functions/Engine_MaterialFunctions02/Utility/Sine_Remapped}.Remapped Sine)) {constB=0.1}
  n15 = TextureCoordinate() {coordinateIndex=1}
  n16 = ScalarParameter() {defaultValue=0.6, parameterName=timespeed}
  n17 = Constant2Vector() {g=1, r=8}
  n18 = Constant2Vector() {g=1, r=16}
  n19 = TextureSample(UVs=Add(
    A=@n15
    B=MaterialFunctionCall!(TimeSpeed=@n16, Jump Amount Per Step=@n17, Number of Steps Per Unit of Time=@n18, Time Offset) {materialFunction=/Engine/Functions/Engine_MaterialFunctions02/SteppingPannerTime}.Result
   ) {constA=512}) {samplerType=SAMPLERTYPE_Masks, texture=/BRCosmetics/Weapons/FORT_Melee/Pickaxe_15/Textures/T_Pickaxe_15_Lightshow2}
  n20 = Add(A=Multiply(A=Clamp(None=@n9), B=@n10), B=Multiply(A=Clamp(None=@n14), B=@n19.R))
  n21 = TextureSample() {texture=/BRCosmetics/Weapons/FORT_Melee/Pickaxe_CandyCane/Textures/T_Pickaxe_Candy_Cane_D}
  n22 = Multiply(A=@n21.RGB, B=@n10) {constB=10}
  n23 = LinearInterpolate(
   A=LinearInterpolate(A=Multiply(A=@n20, B=@n21.RGB), B=Multiply(A=@n22, B=@n19.G), Alpha=ScalarParameter() {defaultValue=1, parameterName=Lightshow1})
   B=Multiply(A=@n22, B=@n19.B)
   Alpha=ScalarParameter() {parameterName=Lightshow2}
  )
  n24 = SetMaterialAttributes!(
   MaterialAttributes
   Base Color=NamedRerouteDeclaration(None=TextureSample() {texture=&T_Pickaxe_Candy_Cane_D}.RGB) {name=baseColor, nodeColor=<rgba 0.5 1 0 1>}
   Metallic=NamedRerouteDeclaration(None=@n1.G) {name=metallic, nodeColor=<rgba 0 0.4141 1 1>}
   Specular=NamedRerouteDeclaration(None=@n1.R) {name=specular, nodeColor=<rgba 0 0.4375 1 1>}
   Roughness=NamedRerouteDeclaration(None=LinearInterpolate(A=@n2, B=@n3, Alpha=@n1.B)) {name=roughness, nodeColor=<rgba 0 1 0.9766 1>}
   Emissive Color=NamedRerouteDeclaration(None=Multiply(A=@n5, B=@n23)) {name=emissive, nodeColor=<rgba 1 0 0.4687 1>}
   Clear Coat=NamedRerouteDeclaration(None=@n1.A) {name=clearCoat, nodeColor=<rgba 0.5234 0 1 1>}
   Clear Coat Roughness=NamedRerouteDeclaration(None=Constant() {r=0.12}) {name=ccRoughness, nodeColor=<rgba 0 1 0.6016 1>}
   Normal=NamedRerouteDeclaration(None=TextureSample() {samplerType=SAMPLERTYPE_Normal, texture=/BRCosmetics/Weapons/FORT_Melee/Pickaxe_CandyCane/Textures/T_Pickaxe_Candy_Cane_N}.RGB) {name=nomral, nodeColor=<rgba 0 1 0.2266 1>}
  ) {attributeSetTypes [
    #1
    #2
    #3
    #4
    #5
    #6
    #7
    #8
   ]}
  MaterialFunctionCall(In=@n24) {materialFunction=&MF_TechArtOverrides_accessories}
  Panner(Coordinate=@n15) {speedX=0.1}
  ScalarParameter() {defaultValue=0.5, parameterName=rough min}
  ScalarParameter() {defaultValue=0.7, parameterName=rough max}
  ScalarParameter() {defaultValue=0.15, parameterName=CCRough}
 ]
}
